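# First-person gunshot sound cue for the high-tier AssaultSurgical auto rifle.
# A Mixer layers two random picks from three waves each: the main fire set, with randomized pitch and boosted volume through a Modulator, and the unmodulated "sw" set.
# Each layer's Quality Level node routes every quality slot to the same Random node.
# It uses first-person projectile attenuation, a ranged-weapon fire sound class and the first-person gunfire concurrency set.
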SoundCue /Game/Sounds/Fort_Weapons/AutoRifle/AssaultSurgical/AssaultSurgical_HighTier/AssaultSurgical_HT_Fire_1P_Cue
root=Mixer_0
nodes [
 {id=Mixer_0, children [Modulator_2, QualityLevel_1], properties {inputVolume [0.8,1]}}
 {id=Modulator_2, children [QualityLevel_0], properties {pitchMax=1.2, pitchMin=1, volumeMax=1.45, volumeMin=1.35}}
 {id=QualityLevel_0, children [Random_2,Random_2,Random_2,Random_2]}
 {id=QualityLevel_1, children [Random_0,Random_0,Random_0,Random_0]}
 {id=Random_0, children [WavePlayer_3,WavePlayer_4,WavePlayer_5], properties {weights [1,1,1]}}
 {id=Random_2, children [WavePlayer_0,WavePlayer_1,WavePlayer_2], properties {weights [1,1,1]}}
 {
  id=WavePlayer_0
  properties {soundWaveAssetPtr=&/AssaultSurgical_HT_Fire_1P_01}
 }
 {
  id=WavePlayer_1
  properties {soundWaveAssetPtr=&/AssaultSurgical_HT_Fire_1P_02}
 }
 {
  id=WavePlayer_2
  properties {soundWaveAssetPtr=&/AssaultSurgical_HT_Fire_1P_03}
 }
 {
  id=WavePlayer_3
  properties {soundWaveAssetPtr=&/AssaultSurgical_HT_Fire_1P_sw_01}
 }
 {
  id=WavePlayer_4
  properties {soundWaveAssetPtr=&/AssaultSurgical_HT_Fire_1P_sw_03}
 }
 {
  id=WavePlayer_5
  properties {soundWaveAssetPtr=&/AssaultSurgical_HT_Fire_1P_sw_02}
 }
]
properties {
 attenuationSettings=&Fort_Weapon_Projectile_FirstPerson_Athena_Att
 bDebug=true
 maxDistance=2097152
 soundClassObject=&Fort_WeaponRanged_Fire_1P_PSM
 concurrencyOverrides {bLimitToOwner=true, maxCount=3, resolutionRule=StopOldest}
 concurrencySet [&Weapon_Gunfire_FirstPerson_Conc]
}
