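# Sound cue for the low-ammo sound of the high-tier Surgical assault rifle.
# It reuses a single low-tier low-ammo wave, passed through a modulator that applies a randomized volume boost and keeps pitch fixed.
# It routes to the degraded ranged-weapon low-ammo sound class and overrides concurrency so new instances are refused at the limit.
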SoundCue /Game/Sounds/Fort_Weapons/AutoRifle/AssaultSurgical/AssaultSurgical_HighTier/AssaultSurgical_HT_LowAmmo
root=Modulator_0
nodes [
 {id=Modulator_0, children [WavePlayer_0], properties {pitchMax=1, pitchMin=1, volumeMax=1.4, volumeMin=1.2}}
 {
  id=WavePlayer_0
  properties {soundWaveAssetPtr=/Game/Sounds/Fort_Weapons/AutoRifle/AssaultSurgical/AssaultSurgical_LowTier/AssaultSurgical_LT_LowAmmo_01}
 }
]
properties {
 bOverrideConcurrency=true
 maxDistance=2097152
 soundClassObject=/Game/Sounds/FortSoundClassesAndModes/Fort_WeaponRanged_LowAmmo_Degraded_PSM
 concurrencyOverrides {resolutionRule=PreventNew}
}
